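; Cosmetic lightning bolt actor, based on FortLightningActor, that shows a brief strike between a start and end point.
; Construction picks a random lifetime and creates dynamic materials for three strike meshes.
; SetupLightning reveals a random mesh, stretches and aims it between the points with random width and spin, then sends time and lifetime to its material.
; A timer hides the mesh and calls CleanupLightning, which DebugStrike's debug flag skips.
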(event Effects|EventOnLightingInitialized)

(event Effects|EventOnLightingCleanup)

(event Custom|DebugStrike
  ;[1 1 1] For Testing
  (Networking|CanExecuteCosmeticEvents
    (:True
      (Variables|Default|SetisDebugMode true)
      (CallFunction|CleanupMesh)
      (CallFunction|GetRandomMesh)
      ;[1 1 1] Transform
      (bind _boltmesh (Variables|Default|GetBoltMesh)
        (:then
          (Transformation|SetRelativeTransform _boltmesh (Math|Transform|MakeTransform (- (Variables|Default|GetStart%20Location) (Transformation|GetActorLocation)) (Math|Rotator|MakeRotfromZ (* (- (Variables|Default|GetEnd%20Location) (Variables|Default|GetStart%20Location)) -1)) (Math|Vector|MakeVector (Math|Float|Lerp 0.0 (Math|Random|RandomFloatinRange 0.65 0.9) (/ (Math|Vector|VectorLength (- (Variables|Default|GetEnd%20Location) (Variables|Default|GetStart%20Location))) 100)) (Math|Float|Lerp 0.0 (Math|Random|RandomFloatinRange 0.65 0.9) (/ (Math|Vector|VectorLength (- (Variables|Default|GetEnd%20Location) (Variables|Default|GetStart%20Location))) 100)) (/ (Math|Vector|VectorLength (- (Variables|Default|GetEnd%20Location) (Variables|Default|GetStart%20Location))) 100))) false false)
          (Transformation|AddLocalRotation (Variables|Default|GetBoltMesh) (Math|Rotator|MakeRotator 0.0 0.0 (Math|Random|RandomFloatinRange 0.0 360)) false false)
          ;[1 1 1] Set material params
          (Rendering|Material|SetScalarParameterValue (Variables|Default|GetBoltMID) "TimestampA" (Utilities|Time|GetGameTimeinSeconds) :!class /Script/Engine.MaterialInstanceDynamic)
          (Rendering|Material|SetScalarParameterValue (Variables|Default|GetBoltMID) "TransitionLength_In" (Variables|Default|GetBoltLifetime) :!class /Script/Engine.MaterialInstanceDynamic)
          (Utilities|Time|SetTimerbyEvent (delegate Custom|EndLightning) (Variables|Default|GetBoltLifetime)))
        (:else)))))

(event Custom|SetupLightning
  (params (Start%20Location :type Vector) (End%20Location :type Vector))
  ;[1 1 1] Call this to begin
  (Networking|CanExecuteCosmeticEvents
    (:True
      (Variables|Default|SetStart%20Location Start%20Location)
      (Variables|Default|SetEnd%20Location End%20Location)
      (CallFunction|GetRandomMesh)
      ;[1 1 1] Transform
      (bind _boltmesh (Variables|Default|GetBoltMesh)
        (:then
          (Transformation|SetRelativeTransform _boltmesh (Math|Transform|MakeTransform (- (Variables|Default|GetStart%20Location) (Transformation|GetActorLocation)) (Math|Rotator|MakeRotfromZ (* (- (Variables|Default|GetEnd%20Location) (Variables|Default|GetStart%20Location)) -1)) (Math|Vector|MakeVector (Math|Float|Lerp 0.0 (Math|Random|RandomFloatinRange 0.65 0.9) (/ (Math|Vector|VectorLength (- (Variables|Default|GetEnd%20Location) (Variables|Default|GetStart%20Location))) 100)) (Math|Float|Lerp 0.0 (Math|Random|RandomFloatinRange 0.65 0.9) (/ (Math|Vector|VectorLength (- (Variables|Default|GetEnd%20Location) (Variables|Default|GetStart%20Location))) 100)) (/ (Math|Vector|VectorLength (- (Variables|Default|GetEnd%20Location) (Variables|Default|GetStart%20Location))) 100))) false false)
          (Transformation|AddLocalRotation (Variables|Default|GetBoltMesh) (Math|Rotator|MakeRotator 0.0 0.0 (Math|Random|RandomFloatinRange 0.0 360)) false false)
          ;[1 1 1] Set material params
          (Rendering|Material|SetScalarParameterValue (Variables|Default|GetBoltMID) "TimestampA" (Utilities|Time|GetGameTimeinSeconds) :!class /Script/Engine.MaterialInstanceDynamic)
          (Rendering|Material|SetScalarParameterValue (Variables|Default|GetBoltMID) "TransitionLength_In" (Variables|Default|GetBoltLifetime) :!class /Script/Engine.MaterialInstanceDynamic)
          (Utilities|Time|SetTimerbyEvent (delegate Custom|EndLightning) (Variables|Default|GetBoltLifetime)))
        (:else)))))

(event Custom|EndLightning
  ;[1 1 1] Done
  (CallFunction|CleanupMesh)
  (if (not (Variables|Default|GetisDebugMode))
    (Effects|CleanupLightning)))

(fn CleanupMesh
  (bind _boltmesh (Variables|Default|GetBoltMesh)
    (:then
      (Development|SetHiddeninGame _boltmesh true false))
    (:else)))

(fn GetRandomMesh
  (bind _boltmeshes (Variables|FortLightningActor|GetBoltMeshes))
  (bind _output_get (LocalVariables|SetRandomMeshToUse (Math|Random|RandomIntegerinRange 0 (Utilities|Array|LastIndex _boltmeshes))))
  (bind _output_get_2 (Variables|Default|SetBoltMesh (Utilities|Array|Get(acopy) _boltmeshes _output_get)))
  (Development|SetHiddeninGame _output_get_2 false false)
  (Variables|Default|SetBoltMID (Utilities|Array|Get(acopy) (Variables|FortLightningActor|GetBoltMIDs) (LocalVariables|GetRandomMeshToUse)))
  (return))

(fn ConstructionScript
  (|CallParentFunction)
  (Networking|CanExecuteCosmeticEvents
    (:True
      (Variables|Default|SetBoltLifetime (Math|Random|RandomFloatinRange (Variables|Default|GetBoltLifetimeMin) (Variables|Default|GetBoltLifetimeMax)))
      (bind _output_get (Variables|FortLightningActor|SetBoltMeshes (Utilities|Array|MakeArray (Variables|Default|GetSm_LightningStrikes_A_01) (Variables|Default|GetSm_LightningStrikes_B_01) (Variables|Default|GetSm_LightningStrikes_C_01))))
      (for _array_element _output_get
        (bind _returnvalue (Rendering|Material|CreateDynamicMaterialInstance _array_element 0 0 "None" :!class /Script/Engine.PrimitiveComponent))
        (Utilities|Array|Add (Variables|FortLightningActor|GetBoltMIDs) _returnvalue)))))
