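# Translucent, two-sided unlit material for a lightning bolt mesh that grows and flickers in world effects.
# A timestamp-driven transition sweeps a reveal along UV V with a brightened leading edge, switching from solid to textured halfway through.
# Sine-perturbed, floored time makes the panning noise jump in steps, feeding a two-color Fresnel-weighted gradient and a rounded, hard-edged mask.
# A V-based taper scales vertex normals, apparently for offsetting the bolt's thickness.
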Material /Game/Effects/Fort_Effects/World/Enviroment/Lightning/M_LightningSystemBolt_01
graph {
 properties {
  blendMode=BLEND_Translucent
  pixelDepthOffsetMode=PDOM_Legacy
  previewMesh=&/Sm_LightningStrikes_A_01
  shadingModel=MSM_Unlit
  twoSided=true
  usedShadingModels=Unlit
 }
 nodes [
  Constant()
  n1 = TextureCoordinate()
  n2 = AppendVector(A=ScalarParameter() {defaultValue=1, parameterName=U Tile}, B=ScalarParameter() {defaultValue=2.2, parameterName=V Tile})
  n3 = Time()
  n4 = Sine(None=@n3) {period=0.2567}
  n5 = Sine(None=@n3) {period=0.7467}
  n6 = Add(A=Multiply(A=Add(A=@n4, B=@n5), B=ScalarParameter() {defaultValue=1.3, parameterName=Oscilation Scale}), B=@n3)
  n7 = TextureSample(UVs=Panner(Coordinate=Multiply(A=@n1, B=@n2), Time=Floor(None=@n6)) {bFractionalPart=true, speedY=3.63}) {samplerType=SAMPLERTYPE_LinearColor, texture=/Game/Effects/Fort_Effects/Textures/Tile/T_AMR_MuzzleRing_01}
  n8 = MaterialFunctionCall(
   TimestampA=ScalarParameter() {parameterName=TimestampA}
   TransitionLength_In=ScalarParameter() {defaultValue=1, group=Timestamp, parameterName=TransitionLength_In}
   Transition In Only?=StaticBool() {value=true}
  ) {materialFunction=&MF_Timestamp_InOut_01}
  n9 = Round(None=@n8.Linear)
  n10 = MaterialFunctionCall(
   Brightness=ScalarParameter() {defaultValue=1.25, parameterName=Fresnel Brightness}
   Exponent=ScalarParameter() {defaultValue=0.65, parameterName=Fresnel Exponent}
  ) {materialFunction=&MF_Fresnel_FX}
  n11 = LinearInterpolate(
   A=VectorParameter() {parameterName=Color A, defaultValue=<rgba 6.6667 0 80 0>}.RGB
   B=VectorParameter() {parameterName=Color B, defaultValue=<rgba 46.9877 15.3 90 0>}.RGB
   Alpha=Multiply(A=LinearInterpolate(B=@n7.G, Alpha=@n9) {constA=1, constAlpha=1}, B=@n10.Fresnel)
  )
  n12 = MaterialFunctionCall(Float2=TextureCoordinate()) {materialFunction=&BreakOutFloat2Components}
  n13 = Add(A=Multiply(A=@n8.Linear, B=ScalarParameter() {defaultValue=5, parameterName=Grow Scale}), B=OneMinus(None=@n12.G))
  n14 = Saturate(None=Floor(None=@n13))
  n15 = ScalarParameter() {defaultValue=-0.2, parameterName=LeadingEdgeBias}
  n16 = OneMinus(None=Floor(None=Add(A=@n13, B=@n15)))
  LinearInterpolate(
   A=@n11
   B=Multiply(A=@n11, B=ScalarParameter() {defaultValue=100, parameterName=Leading Edge Scale})
   Alpha=Multiply(A=@n14, B=Saturate(None=@n16))
  )
  n17 = MaterialFunctionCall(Float2=TextureCoordinate()) {materialFunction=&BreakOutFloat2Components}
  n18 = ScalarParameter() {defaultValue=0.2, parameterName=TwistAmount}
  MaterialFunctionCall(X=@n17.R, Y=Add(A=Multiply(A=@n17.R, B=@n18), B=@n17.G)) {materialFunction=&MakeFloat2}
  MaterialFunctionCall(Speed=ScalarParameter() {defaultValue=0.2, parameterName=DebugSpeed}) {materialFunction=&DebugTimeSine}
  Multiply(
   A=LinearInterpolate(A=ScalarParameter() {parameterName=Taper Scale A}, B=ScalarParameter() {defaultValue=-10, parameterName=Taper Scale B}, Alpha=@n12.G)
   B=VertexNormalWS()
  )
  n19 = LinearInterpolate(B=@n7.R, Alpha=@n9) {constA=1, constAlpha=1}
  n20 = Multiply(A=Multiply(A=@n19, B=Add(A=@n19, B=@n8.Linear)), B=@n14)
  Round(None=Subtract(A=Saturate(None=@n20), B=OneMinus(None=@n10.Fresnel Double Sided)))
 ]
}
